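# Sound cue for picking up the SteamPunk launcher weapon, classed as a loot collect sound.
# A single wave player plays the launcher pickup sample with slightly lowered pitch and reduced volume.
# It uses the shared weapon pickup attenuation and overrides concurrency to cap simultaneous plays, stopping the oldest.
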SoundCue /Game/Sounds/Fort_Weapons/SteamPunk/Launcher/SP_Launcher_Pickup_Cue
root=WavePlayer_2
nodes [{id=WavePlayer_2, properties {soundWaveAssetPtr=&/SP_Launcher_Pickup_01}}]
properties {
 attenuationSettings=&Fort_Weapon_Pickup_Att
 bOverrideConcurrency=true
 maxDistance=3500
 pitchMultiplier=0.9
 soundClassObject=&Fort_Loot_Collect
 volumeMultiplier=0.2308
 concurrencyOverrides {maxCount=3, resolutionRule=StopOldest}
}
